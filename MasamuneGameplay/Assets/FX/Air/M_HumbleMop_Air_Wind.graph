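# Translucent, unlit, two-sided Niagara material for air/wind VFX on sprites, ribbons and meshes.
# Two panning noise masks, UV gradients, fresnel, depth and camera fades shape soft-edged opacity.
# Emissive blends dark and light colors by mask and face side, tinted by particle color.
# Niagara dynamic parameters drive UVs, time, erosion and a V-weighted trail bend, added to noise-based normal displacement for world position offset.
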Material /MasamuneGameplay/Assets/FX/Air/M_HumbleMop_Air_Wind
graph {
 properties {
  bCastRayTracedShadows=false
  bUsedWithNanite=true
  bUsedWithNiagaraMeshParticles=true
  bUsedWithNiagaraRibbons=true
  bUsedWithNiagaraSprites=true
  blendMode=BLEND_Translucent
  pixelDepthOffsetMode=PDOM_Legacy
  shadingModel=MSM_Unlit
  twoSided=true
  usedShadingModels=Unlit
 }
 nodes [
  n1 = TextureSampleParameter2D(UVs=NamedRerouteUsage()) {group=WPO, parameterName=WPO Noise Mask, texture=&T_Smooth_tile}
  LinearInterpolate(
   A=@n1.RGB
   B=TextureSampleParameter2D(UVs=NamedRerouteUsage()) {group=WPO, parameterName=WPO Noise Mask, texture=&T_Smooth_tile}.RGB
   Alpha=ScalarParameter() {defaultValue=0.5, group=WPO, parameterName=WPO Noise Mask Bias}
  )
  MaterialFunctionCall!(
   TangentU_WS
   TangentV_WS
   WidthBase
   WidthTip
   WorldPosition
   UVs for Projection
   UVs for Thickness
   UVs for texturing
   Expand U or V UV Channel
   DeriveNormalZ
   AdditionalNormal
   AngleCorrectedNormals
   FlattenPixelNormal
   UVs For NormalShading
  ) {materialFunction=/Game/Effects/Niagara/Modules/FollowingDeform/Material/SplineThickenWithTangentInput}
  n2 = ParticleColor()
  n3 = VectorParameter() {group=Color, parameterName=Dark Color, defaultValue=<rgba 0.1624 0.4354 0.117 1>}
  n4 = VectorParameter() {group=Color, parameterName=Light Color, defaultValue=<rgba 0.3991 0.8438 0.3252 1>}
  n5 = ScalarParameter() {defaultValue=0.25, group=Color, parameterName=Dark Color End}
  n6 = ScalarParameter() {defaultValue=0.75, group=Color, parameterName=Dark Color Start}
  n7 = Subtract(A=@n6) {constB=0.0001}
  n8 = ConstantBiasScale(None=TwoSidedSign())
  n9 = OneMinus(None=@n8)
  n10 = ScalarParameter() {group=Color, parameterName=Swap Dark Light}
  n11 = Multiply(
   A=SmoothStep(Min=Min(A=@n5, B=@n7) {constB=1}, Max=@n6, Value=NamedRerouteUsage())
   B=LinearInterpolate(Alpha=LinearInterpolate(A=@n8, B=@n9, Alpha=@n10)) {constA=0.15}
  )
  MaterialFunctionCall(Emissive=Multiply(A=@n2.RGB, B=LinearInterpolate(A=@n3.RGB, B=@n4.RGB, Alpha=@n11) {constA=0.75})) {materialFunction=&MF_FX_EmissiveAdjustments}
  n12 = ScalarParameter() {defaultValue=0.5, parameterName=Opacity Softness}
  n13 = NamedRerouteUsage()
  n14 = NamedRerouteUsage()
  n15 = Multiply(A=Multiply(A=Multiply(A=@n13, B=@n14), B=NamedRerouteUsage()), B=NamedRerouteUsage())
  n16 = NamedRerouteUsage()
  n17 = Multiply(
   A=Multiply(A=Multiply(A=@n15, B=@n16), B=NamedRerouteUsage())
   B=MaterialFunctionCall(Distance=ScalarParameter() {defaultValue=50, group=Opacity, parameterName=Camera Fade Distance}) {materialFunction=&Near_Camera_Fade}.Result
  )
  n18 = Multiply(A=@n17, B=VertexColor().A)
  n19 = ScalarParameter() {group=Opacity, parameterName=Use Vertex Alpha}
  Multiply(A=SmoothStep(Max=Max(A=@n12) {constB=0.0001}, Value=LinearInterpolate(A=@n17, B=@n18, Alpha=@n19)), B=ParticleColor().A)
  n20 = DynamicParameter() {desc=Parameters set by Niagara, parameterIndex=2, paramNames [X, Y, Z, WPO Trail Intensity]}
  n21 = DynamicParameter() {parameterIndex=1, defaultValue=<rgba 0 1 1 1>, paramNames [Time, Erosion Opacity, WPO Trail Max Offset, WPO Trail Bend Power]}
  n22 = Clamp(
   None=MaterialFunctionCall(X=@n20.X, Y=@n20.X, Z=@n20.X) {materialFunction=&MakeFloat3}.Result
   Min=Multiply(A=@n21.WPO Trail Max Offset) {constA=-1, constB=-1}
   Max=@n21.WPO Trail Max Offset
  ) {maxDefault=50, minDefault=-50}
  n23 = Saturate(None=ComponentMask(None=TextureCoordinate()) {g=true})
  n24 = Multiply(A=Power(Base=OneMinus(None=@n23), Exp=@n21.WPO Trail Bend Power), B=@n20.WPO Trail Intensity)
  n25 = VectorParameter() {group=WPO, parameterName=WPO Trail Multiplier, defaultValue=<rgba 1 1 1 0>}
  NamedRerouteDeclaration(None=Multiply(A=LinearInterpolate(B=@n22, Alpha=@n24), B=Multiply(A=@n25.RGB, B=@n25.RGB))) {name=WPO Trail, nodeColor=<rgba 0 1 0.25 1>}
  n26 = LinearInterpolate(
   A=TextureSampleParameter2D(UVs=NamedRerouteUsage()) {group=Opacity, parameterName=Noise Mask A, texture=&T_ShieldCaustics}.RGB
   B=TextureSampleParameter2D(UVs=NamedRerouteUsage()) {group=Opacity, parameterName=Noise Mask B, texture=&T_ShieldCaustics}.RGB
   Alpha=ScalarParameter() {defaultValue=0.5, group=Opacity, parameterName=Noise Mask Bias}
  )
  n27 = ScalarParameter() {defaultValue=1, group=Opacity, parameterName=Noise Mask Exponent}
  NamedRerouteDeclaration(None=LinearInterpolate(
    A=ScalarParameter() {group=Opacity, parameterName=Noise Mask Min}
    B=ScalarParameter() {defaultValue=1, group=Opacity, parameterName=Noise Mask Max}
    Alpha=Power(Base=@n26, Exp=@n27)
   )) {name=Noise Mask, nodeColor=<rgba 1 0 0.5625 1>}
  n28 = DynamicParameter() {
   desc=Niagara index 1
   parameterIndex=1
   defaultValue=<rgba 0 1 1 1>
   paramNames [Time, Erosion Opacity, WPO Trail Max Offset, WPO Trail Bend Power]
  }
  NamedRerouteDeclaration(None=LinearInterpolate(
    A=Time()
    B=@n28.Time
    Alpha=ScalarParameter() {desc=Use the Time input from Niagara Dynamic Material Parameter, parameterName=Use Niagara DMP Time}
   )) {name=Time, nodeColor=<rgba 0.9453 1 0 1>}
  n29 = DynamicParameter() {defaultValue=<rgba 1 1 0 0>, paramNames [U Tile Multiplier, V Tile Multiplier, U Offset, V Offset]}
  NamedRerouteDeclaration(None=MaterialFunctionCall(X=@n29.U Offset, Y=@n29.U Offset) {materialFunction=&MakeFloat2}.Result) {name=DP UV Offset, nodeColor=<rgba 0 1 0.0391 1>}
  NamedRerouteDeclaration(None=AppendVector(A=@n29.U Tile Multiplier, B=@n29.U Tile Multiplier)) {name=DP UV Tile, nodeColor=<rgba 0.4062 1 0 1>}
  n30 = VectorParameter() {group=Opacity, parameterName="Noise Mask UV A Tile\\Speed", defaultValue=<rgba 1 1 0.2 1>}
  n31 = NamedRerouteUsage()
  NamedRerouteDeclaration(None=Multiply(A=ComponentMask(None=@n30.RGB) {g=true, r=true}, B=@n31)) {name=UV A Tile, nodeColor=<rgba 1 1 0 1>}
  n32 = VectorParameter() {group=Opacity, parameterName="Noise Mask UV B Tile\\Speed", defaultValue=<rgba 1 1 -0.25 0.8>}
  NamedRerouteDeclaration(None=Multiply(A=ComponentMask(None=@n32.RGB) {g=true, r=true}, B=@n31)) {name=UV B Tile, nodeColor=<rgba 1 1 0 1>}
  n33 = LinearInterpolate(A=TextureCoordinate(), B=TextureCoordinate() {coordinateIndex=1}, Alpha=ScalarParameter() {parameterName=Use UV2 for Noise Texture})
  n34 = Add(A=@n33, B=NamedRerouteUsage())
  n35 = NamedRerouteUsage()
  NamedRerouteDeclaration(None=Panner(Coordinate=Multiply(A=@n34, B=@n35), Time=NamedRerouteUsage(), Speed=NamedRerouteUsage()) {speedX=0.2, speedY=1}) {name=UV A, nodeColor=<rgba 0.4766 1 0 1>}
  n36 = Add(A=@n33, B=NamedRerouteUsage())
  n37 = NamedRerouteUsage()
  NamedRerouteDeclaration(None=Panner(Coordinate=Multiply(A=@n36, B=@n37), Time=NamedRerouteUsage(), Speed=NamedRerouteUsage()) {speedX=-0.25, speedY=0.8}) {name=UV B, nodeColor=<rgba 1 0 0.3516 1>}
  NamedRerouteDeclaration(None=@n28.Erosion Opacity) {name=DP Erosion Alpha, nodeColor=<rgba 0 0.8828 1 1>}
  n38 = VectorParameter() {group=WPO, parameterName="WPO Noise Mask UV A Tile\\Speed", defaultValue=<rgba 1 1 0.2 1>}
  NamedRerouteDeclaration(None=MaterialFunctionCall(X=@n38.B, Y=@n38.B) {materialFunction=&MakeFloat2}.Result) {name=WPO UV A Pan Speed, nodeColor=<rgba 1 1 0 1>}
  n39 = VectorParameter() {group=WPO, parameterName="WPO Noise Mask UV B Tile\\Speed", defaultValue=<rgba 1 1 -0.25 0.8>}
  NamedRerouteDeclaration(None=MaterialFunctionCall(X=@n39.B, Y=@n39.B) {materialFunction=&MakeFloat2}.Result) {name=WPO UV B Pan Speed, nodeColor=<rgba 1 1 0 1>}
  n40 = ScalarParameter() {defaultValue=1, group=WPO, parameterName=WPO Noise Mask Exponent}
  NamedRerouteDeclaration(None=LinearInterpolate(
    A=ScalarParameter() {group=WPO, parameterName=WPO Noise Mask Min}
    B=ScalarParameter() {defaultValue=1, group=WPO, parameterName=WPO Noise Mask Max}
    Alpha=Power(Base=@n1.RGB, Exp=@n40)
   )) {name=WPO Noise Mask, nodeColor=<rgba 1 0.2812 0 1>}
  n41 = NamedRerouteUsage()
  NamedRerouteDeclaration(None=Multiply(A=ComponentMask(None=@n38.RGB) {g=true, r=true}, B=@n41)) {name=WPO UV A Tile, nodeColor=<rgba 1 1 0 1>}
  NamedRerouteDeclaration(None=Multiply(A=ComponentMask(None=@n39.RGB) {g=true, r=true}, B=@n41)) {name=WPO UV B Tile, nodeColor=<rgba 1 1 0 1>}
  n42 = ScalarParameter() {defaultValue=10, parameterName=Depth Fade Distance}
  NamedRerouteDeclaration(None=ShadingPathSwitch(Default=DepthFade(FadeDistance=@n42), Mobile=Constant() {r=1}) {desc=Migrated from FeatureLevelSwitch}) {name=Depth Fade, nodeColor=<rgba 0.6875 0 1 1>}
  n43 = CameraVectorWS()
  n44 = VertexNormalWS()
  n45 = Power(Base=Abs(None=DotProduct(A=@n43, B=@n44)), Exp=ScalarParameter() {defaultValue=4, parameterName=Fresnel Power}) {constExponent=4}
  n46 = ScalarParameter() {defaultValue=4, parameterName=Fresnel Multiplier}
  NamedRerouteDeclaration(None=Saturate(None=Multiply(A=@n45, B=@n46) {constB=4})) {name=Fresnel, nodeColor=<rgba 1 0 0.1875 1>}
  n47 = Multiply(A=VertexNormalWS(), B=NamedRerouteUsage())
  n48 = ScalarParameter() {group=WPO, parameterName=WPO Amount}
  NamedRerouteDeclaration(None=Add(A=Multiply(A=@n47, B=@n48) {constB=7}, B=NamedRerouteUsage())) {name=WPO, nodeColor=<rgba 0.7109 0 1 1>}
  n49 = Saturate(None=ScalarParameter() {group=UV Gradient, parameterName=U Gradient})
  n50 = TextureCoordinate()
  n51 = Constant() {r=0.5}
  n52 = SphereMask(
   A=Power(
    Base=MaterialFunctionCall(Float2=@n50) {materialFunction=&BreakOutFloat2Components}.R
    Exp=ScalarParameter() {defaultValue=1, group=UV Gradient, parameterName=U Gradient Position Bias}
   )
   B=@n51
   Radius=@n51
   Hardness=ScalarParameter() {group=UV Gradient, parameterName=U Gradient Hardness}
  ) {attenuationRadius=0.5, hardnessPercent=0}
  n53 = ScalarParameter() {defaultValue=1, group=UV Gradient, parameterName=U Gradient Power}
  NamedRerouteDeclaration(None=LinearInterpolate(A=OneMinus(None=@n49), Alpha=Power(Base=@n52, Exp=@n53) {constExponent=1.5})) {name=U Gradient, nodeColor=<rgba 0.8516 0 1 1>}
  n54 = Saturate(None=ScalarParameter() {group=UV Gradient, parameterName=V Gradient})
  n55 = TextureCoordinate()
  n56 = Constant() {r=0.5}
  n57 = SphereMask(
   A=Power(
    Base=MaterialFunctionCall(Float2=@n55) {materialFunction=&BreakOutFloat2Components}.G
    Exp=ScalarParameter() {defaultValue=1, group=UV Gradient, parameterName=V Gradient Position Bias}
   ) {constExponent=1}
   B=@n56
   Radius=@n56
   Hardness=ScalarParameter() {group=UV Gradient, parameterName=V Gradient Hardness}
  ) {attenuationRadius=0.5, hardnessPercent=0}
  n58 = ScalarParameter() {defaultValue=1, group=UV Gradient, parameterName=V Gradient Power}
  NamedRerouteDeclaration(None=LinearInterpolate(A=OneMinus(None=@n54), Alpha=Power(Base=@n57, Exp=@n58) {constExponent=2.2})) {name=V Gradient, nodeColor=<rgba 0 0.5547 1 1>}
  NamedRerouteDeclaration(None=MaterialFunctionCall(X=@n30.B, Y=@n30.B) {materialFunction=&MakeFloat2}.Result) {name=UV A Pan Speed, nodeColor=<rgba 1 1 0 1>}
  NamedRerouteDeclaration(None=MaterialFunctionCall(X=@n32.B, Y=@n32.B) {materialFunction=&MakeFloat2}.Result) {name=UV B Pan Speed, nodeColor=<rgba 1 1 0 1>}
  NamedRerouteUsage()
  Saturate()
  ScalarParameter() {defaultValue=1, group=Opacity, parameterName=Noise Multiplier}
  n59 = ScalarParameter() {defaultValue=1, parameterName=Refraction}
  n60 = MaterialFunctionCall(Distance=ScalarParameter() {defaultValue=200, parameterName=CameraFadeDistance}) {materialFunction=&Near_Camera_Fade}
  n61 = LinearInterpolate(
   B=StaticSwitchParameter(True=LinearInterpolate(B=@n59, Alpha=@n60.Result) {constA=1}, False=@n59) {group=DetailedGradientMask, parameterName=UseCameraNearFade}
   Alpha=Desaturation(Fraction=Constant() {r=1})
  ) {constA=1, constB=1.2}
  n62 = Constant() {r=1}
  n63 = Constant() {r=1}
  ShadingPathSwitch(
   Default=StaticSwitchParameter(True=QualitySwitch(Default=@n61, Low=@n62), False=@n63) {defaultValue=true, parameterName=Use_Distortion}
   Forward=@n63
   Mobile=@n63
  )
  n64 = Multiply(B=@n2.RGB)
  n65 = Clamp(None=Multiply(B=@n64))
  n66 = ScalarParameter() {defaultValue=1.5, group=Color, parameterName=HDRMult}
  ShadingPathSwitch(
   Default=MaterialFunctionCall(LDR Input=@n65, HDR Tint=Multiply(A=@n64, B=@n66)) {materialFunction=&DeriveHDRfromLDR}.RemappedRange
   Mobile=@n65
  ) {desc=Migrated from FeatureLevelSwitch}
  TextureCoordinate() {coordinateIndex=1}
 ]
}
