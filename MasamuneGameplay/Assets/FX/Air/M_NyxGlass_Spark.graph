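# Unlit translucent material for Niagara sprite particles, named as a Nyx Glass spark, built on a soft flare texture parameter.
# The texture's red channel is shaped by MF_FX_Refine with TRefine, multiplied by particle alpha and depth-faded for opacity.
# For color, the offset texture RGB is normalized, desaturated by a parameter, tinted by particle color and passed through MF_FX_EmissiveAdjustments.
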Material /MasamuneGameplay/Assets/FX/Air/M_NyxGlass_Spark
graph {
 properties {bUsedWithNiagaraSprites=true, blendMode=BLEND_Translucent, pixelDepthOffsetMode=PDOM_Legacy, shadingModel=MSM_Unlit, usedShadingModels=Unlit}
 nodes [
  n1 = TextureSampleParameter2D() {parameterName=Tex, texture=/Game/Effects/Fort_Effects/Textures/Flares/T_Flare_Soft}
  n2 = VectorParameter() {parameterName=TRefine, defaultValue=<rgba 5 1 0.2 0.5>}
  n3 = ParticleColor()
  DepthFade(
   Opacity=Multiply(
    A=MaterialFunctionCall(Texture=@n1.R, RGB=@n2.RGB, A=@n2.RGB) {materialFunction=&MF_FX_Refine}.Result
    B=@n3.A
   )
   FadeDistance=ScalarParameter() {defaultValue=2, parameterName=DF}
  )
  n4 = Normalize(VectorInput=Add(A=@n1.RGB) {constB=0.01})
  n5 = ScalarParameter() {parameterName=Desat}
  MaterialFunctionCall(Emissive=Multiply(A=Desaturation(None=@n4, Fraction=@n5), B=@n3.RGB)) {materialFunction=&MF_FX_EmissiveAdjustments}
 ]
}
